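# Sound cue for quiet one-shot evening forest bird calls at The Lair POI in Asteria, built from generic Apollo biome bird waves.
# A quality-level switch sends the first two levels to a five-wave random pool and the last two to a three-wave subset.
# Each pool has a modulator that widens pitch and volume variation downward.
# It uses tile-ambient bird attenuation and the ambient one-shots sound class.
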SoundCue /Game/Sounds/Asteria/POI/The_Lair/Spots/Spot_Lair_Forest_Evening_Bird_Cue
root=QualityLevel_0
nodes [
 {id=Modulator_0, children [Random_1], properties {pitchMin=0.3, volumeMin=0.3}}
 {id=Modulator_1, children [Random_3], properties {pitchMin=0.3, volumeMin=0.3}}
 {id=QualityLevel_0, children [Modulator_0,Modulator_0,Modulator_1,Modulator_1]}
 {id=Random_1, children [WavePlayer_0, WavePlayer_7, WavePlayer_8, WavePlayer_9, WavePlayer_10], properties {weights [1,1,1,1,1]}}
 {id=Random_3, children [WavePlayer_11, WavePlayer_20, WavePlayer_21], properties {weights [1,1,1]}}
 {id=WavePlayer_0, properties {soundWaveAssetPtr=/Game/Sounds/Apollo/Biomes/Generic/Biome_Generic_Bird_A_02}}
 {id=WavePlayer_10, properties {soundWaveAssetPtr=/Game/Sounds/Apollo/Biomes/Generic/Biome_Generic_Bird_A_06}}
 {id=WavePlayer_11, properties {soundWaveAssetPtr=&Biome_Generic_Bird_A_02}}
 {id=WavePlayer_20, properties {soundWaveAssetPtr=/Game/Sounds/Apollo/Biomes/Generic/Biome_Generic_Bird_A_05}}
 {id=WavePlayer_21, properties {soundWaveAssetPtr=&Biome_Generic_Bird_A_06}}
 {id=WavePlayer_7, properties {soundWaveAssetPtr=/Game/Sounds/Apollo/Biomes/Generic/Biome_Generic_Bird_A_01}}
 {id=WavePlayer_8, properties {soundWaveAssetPtr=/Game/Sounds/Apollo/Biomes/Generic/Biome_Generic_Bird_A_04}}
 {id=WavePlayer_9, properties {soundWaveAssetPtr=&Biome_Generic_Bird_A_05}}
]
properties {
 attenuationSettings=&Fort_TileAmbient_OneShot_Bird_Att
 bEnableBusSends=false
 maxDistance=8005
 soundClassObject=&Fort_Ambient_OneShots
 volumeMultiplier=0.125
 preEffectBusSends [{}]
}
