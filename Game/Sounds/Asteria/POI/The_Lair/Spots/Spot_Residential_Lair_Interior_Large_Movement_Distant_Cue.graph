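# Sound cue for spot ambient one-shots of distant large movement inside The Lair POI in Asteria.
# A Random node picks with equal weight among distant shuffle wave variations, and a Modulator randomly lowers pitch and volume slightly for variety.
# The cue plays at reduced volume on the ambient one-shot sound class, using the tile ambient one-shot attenuation.
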SoundCue /Game/Sounds/Asteria/POI/The_Lair/Spots/Spot_Residential_Lair_Interior_Large_Movement_Distant_Cue
root=Modulator_0
nodes [
 {id=Modulator_0, children [Random_1], properties {pitchMax=1, pitchMin=0.8, volumeMax=1, volumeMin=0.7}}
 {
  id=Random_1
  children [
   WavePlayer_3
   WavePlayer_4
   WavePlayer_5
   WavePlayer_6
   WavePlayer_7
   WavePlayer_8
   WavePlayer_9
   WavePlayer_10
   WavePlayer_22
   WavePlayer_23
   WavePlayer_24
   WavePlayer_25
   WavePlayer_30
   WavePlayer_31
   WavePlayer_32
   WavePlayer_33
   WavePlayer_34
   WavePlayer_35
   WavePlayer_36
   WavePlayer_37
   WavePlayer_38
   WavePlayer_39
   WavePlayer_40
   WavePlayer_41
   WavePlayer_42
   WavePlayer_43
  ]
  properties {weights [1,1,1,1,1,1,1,1,1,1,1,1,1,1,1,1,1,1,1,1,1,1,1,1,1,1]}
 }
 {id=WavePlayer_10, properties {soundWaveAssetPtr=&/Estate_Lair_Int_Distant_Shuffle_09}}
 {id=WavePlayer_22, properties {soundWaveAssetPtr=&/Estate_Lair_Int_Distant_Shuffle_10}}
 {id=WavePlayer_23, properties {soundWaveAssetPtr=&/Estate_Lair_Int_Distant_Shuffle_11}}
 {id=WavePlayer_24, properties {soundWaveAssetPtr=&/Estate_Lair_Int_Distant_Shuffle_12}}
 {id=WavePlayer_25, properties {soundWaveAssetPtr=&/Estate_Lair_Int_Distant_Shuffle_13}}
 {id=WavePlayer_3, properties {soundWaveAssetPtr=&/Estate_Lair_Int_Distant_Shuffle_02}}
 {id=WavePlayer_30, properties {soundWaveAssetPtr=&/Estate_Lair_Int_Distant_Shuffle_14}}
 {id=WavePlayer_31, properties {soundWaveAssetPtr=&/Estate_Lair_Int_Distant_Shuffle_15}}
 {id=WavePlayer_32, properties {soundWaveAssetPtr=&/Estate_Lair_Int_Distant_Shuffle_16}}
 {id=WavePlayer_33, properties {soundWaveAssetPtr=&/Estate_Lair_Int_Distant_Shuffle_17}}
 {id=WavePlayer_34, properties {soundWaveAssetPtr=&/Estate_Lair_Int_Distant_Shuffle_18}}
 {id=WavePlayer_35, properties {soundWaveAssetPtr=&/Estate_Lair_Int_Distant_Shuffle_19}}
 {id=WavePlayer_36, properties {soundWaveAssetPtr=&/Estate_Lair_Int_Distant_Shuffle_20}}
 {id=WavePlayer_37, properties {soundWaveAssetPtr=&/Estate_Lair_Int_Distant_Shuffle_21}}
 {id=WavePlayer_38, properties {soundWaveAssetPtr=&/Estate_Lair_Int_Distant_Shuffle_22}}
 {id=WavePlayer_39, properties {soundWaveAssetPtr=&/Estate_Lair_Int_Distant_Shuffle_23}}
 {id=WavePlayer_4, properties {soundWaveAssetPtr=&/Estate_Lair_Int_Distant_Shuffle_03}}
 {id=WavePlayer_40, properties {soundWaveAssetPtr=&/Estate_Lair_Int_Distant_Shuffle_24}}
 {id=WavePlayer_41, properties {soundWaveAssetPtr=&/Estate_Lair_Int_Distant_Shuffle_25}}
 {id=WavePlayer_42, properties {soundWaveAssetPtr=&/Estate_Lair_Int_Distant_Shuffle_26}}
 {id=WavePlayer_43, properties {soundWaveAssetPtr=&/Estate_Lair_Int_Distant_Shuffle_27}}
 {id=WavePlayer_5, properties {soundWaveAssetPtr=&/Estate_Lair_Int_Distant_Shuffle_04}}
 {id=WavePlayer_6, properties {soundWaveAssetPtr=&/Estate_Lair_Int_Distant_Shuffle_05}}
 {id=WavePlayer_7, properties {soundWaveAssetPtr=&/Estate_Lair_Int_Distant_Shuffle_06}}
 {id=WavePlayer_8, properties {soundWaveAssetPtr=&/Estate_Lair_Int_Distant_Shuffle_07}}
 {id=WavePlayer_9, properties {soundWaveAssetPtr=&/Estate_Lair_Int_Distant_Shuffle_08}}
]
properties {
 attenuationSettings=/Game/Sounds/Attenuation/Fort_TileAmbient_OneShot_Att
 maxDistance=8000
 soundClassObject=&Fort_Ambient_OneShots
 volumeMultiplier=0.376
}
