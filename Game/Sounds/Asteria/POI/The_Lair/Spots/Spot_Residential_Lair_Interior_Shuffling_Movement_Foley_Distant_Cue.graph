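# Sound cue for distant interior shuffling-movement foley, an ambient one-shot spot sound for The Lair POI.
# A Random node picks with equal weight among eighteen wave variations, all borrowed from the Beach Resort POI's residential interior spot set.
# A Modulator at the root randomizes pitch downward and varies volume, under a reduced cue volume multiplier.
# Playback uses the ambient one-shot sound class and the tile ambient one-shot attenuation.
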SoundCue /Game/Sounds/Asteria/POI/The_Lair/Spots/Spot_Residential_Lair_Interior_Shuffling_Movement_Foley_Distant_Cue
root=Modulator_0
nodes [
 {id=Modulator_0, children [Random_0], properties {pitchMax=1, pitchMin=0.8, volumeMax=1, volumeMin=0.7}}
 {
  id=Random_0
  children [
   WavePlayer_0
   WavePlayer_1
   WavePlayer_2
   WavePlayer_11
   WavePlayer_12
   WavePlayer_13
   WavePlayer_14
   WavePlayer_15
   WavePlayer_16
   WavePlayer_17
   WavePlayer_18
   WavePlayer_19
   WavePlayer_20
   WavePlayer_21
   WavePlayer_26
   WavePlayer_27
   WavePlayer_28
   WavePlayer_29
  ]
  properties {weights [1,1,1,1,1,1,1,1,1,1,1,1,1,1,1,1,1,1]}
 }
 {
  id=WavePlayer_0
  properties {soundWaveAssetPtr=/Game/Sounds/Asteria/POI/Beach_Resort/Spots/Spot_Residential_Interior_Shuffling_Movement_Foley_Distant_18}
 }
 {
  id=WavePlayer_1
  properties {soundWaveAssetPtr=/Game/Sounds/Asteria/POI/Beach_Resort/Spots/Spot_Residential_Interior_Shuffling_Movement_Foley_Distant_01}
 }
 {
  id=WavePlayer_11
  properties {soundWaveAssetPtr=/Game/Sounds/Asteria/POI/Beach_Resort/Spots/Spot_Residential_Interior_Shuffling_Movement_Foley_Distant_03}
 }
 {
  id=WavePlayer_12
  properties {soundWaveAssetPtr=/Game/Sounds/Asteria/POI/Beach_Resort/Spots/Spot_Residential_Interior_Shuffling_Movement_Foley_Distant_04}
 }
 {
  id=WavePlayer_13
  properties {soundWaveAssetPtr=/Game/Sounds/Asteria/POI/Beach_Resort/Spots/Spot_Residential_Interior_Shuffling_Movement_Foley_Distant_05}
 }
 {
  id=WavePlayer_14
  properties {soundWaveAssetPtr=/Game/Sounds/Asteria/POI/Beach_Resort/Spots/Spot_Residential_Interior_Shuffling_Movement_Foley_Distant_06}
 }
 {
  id=WavePlayer_15
  properties {soundWaveAssetPtr=/Game/Sounds/Asteria/POI/Beach_Resort/Spots/Spot_Residential_Interior_Shuffling_Movement_Foley_Distant_07}
 }
 {
  id=WavePlayer_16
  properties {soundWaveAssetPtr=/Game/Sounds/Asteria/POI/Beach_Resort/Spots/Spot_Residential_Interior_Shuffling_Movement_Foley_Distant_08}
 }
 {
  id=WavePlayer_17
  properties {soundWaveAssetPtr=/Game/Sounds/Asteria/POI/Beach_Resort/Spots/Spot_Residential_Interior_Shuffling_Movement_Foley_Distant_09}
 }
 {
  id=WavePlayer_18
  properties {soundWaveAssetPtr=/Game/Sounds/Asteria/POI/Beach_Resort/Spots/Spot_Residential_Interior_Shuffling_Movement_Foley_Distant_10}
 }
 {
  id=WavePlayer_19
  properties {soundWaveAssetPtr=/Game/Sounds/Asteria/POI/Beach_Resort/Spots/Spot_Residential_Interior_Shuffling_Movement_Foley_Distant_11}
 }
 {
  id=WavePlayer_2
  properties {soundWaveAssetPtr=/Game/Sounds/Asteria/POI/Beach_Resort/Spots/Spot_Residential_Interior_Shuffling_Movement_Foley_Distant_02}
 }
 {
  id=WavePlayer_20
  properties {soundWaveAssetPtr=/Game/Sounds/Asteria/POI/Beach_Resort/Spots/Spot_Residential_Interior_Shuffling_Movement_Foley_Distant_12}
 }
 {
  id=WavePlayer_21
  properties {soundWaveAssetPtr=/Game/Sounds/Asteria/POI/Beach_Resort/Spots/Spot_Residential_Interior_Shuffling_Movement_Foley_Distant_13}
 }
 {
  id=WavePlayer_26
  properties {soundWaveAssetPtr=/Game/Sounds/Asteria/POI/Beach_Resort/Spots/Spot_Residential_Interior_Shuffling_Movement_Foley_Distant_14}
 }
 {
  id=WavePlayer_27
  properties {soundWaveAssetPtr=/Game/Sounds/Asteria/POI/Beach_Resort/Spots/Spot_Residential_Interior_Shuffling_Movement_Foley_Distant_15}
 }
 {
  id=WavePlayer_28
  properties {soundWaveAssetPtr=/Game/Sounds/Asteria/POI/Beach_Resort/Spots/Spot_Residential_Interior_Shuffling_Movement_Foley_Distant_16}
 }
 {
  id=WavePlayer_29
  properties {soundWaveAssetPtr=/Game/Sounds/Asteria/POI/Beach_Resort/Spots/Spot_Residential_Interior_Shuffling_Movement_Foley_Distant_17}
 }
]
properties {
 attenuationSettings=/Game/Sounds/Attenuation/Fort_TileAmbient_OneShot_Att
 maxDistance=8000
 soundClassObject=&Fort_Ambient_OneShots
 volumeMultiplier=0.451
}
